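# Default Niagara material for geometry cache meshes, flagged for use with geometry caches and shaded as Default Lit.
# It holds a vector parameter named Color and a scalar constant, with no links from either node to material outputs shown.
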Material /Niagara/DefaultAssets/GeometryCache/M_DefaultGeometryCacheMaterial
graph {
 properties {bUsedWithGeometryCache=true, pixelDepthOffsetMode=PDOM_Legacy, usedShadingModels=Default Lit}
 nodes [
  Constant() {r=1}
  VectorParameter() {parameterName=Color}
 ]
}
